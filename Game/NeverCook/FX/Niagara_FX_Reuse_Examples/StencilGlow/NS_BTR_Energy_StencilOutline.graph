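# Niagara system in a reuse-examples folder that layers three sprite emitters, BTR_Energy, BTR_StencilOutline and BTR_StencilGlow, into an energy effect named for a stencil outline and glow.
# Each emitter burst-spawns fixed-size, non-uniform sprites with per-particle material randoms and does not kill them when their lifetime elapses.
# The glow layer alone adds ScaleColor with boosted alpha.
# No user parameters or motion modules are present, so the look likely comes from sprite materials, which are not shown.
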NiagaraSystem /Game/NeverCook/FX/Niagara_FX_Reuse_Examples/StencilGlow/NS_BTR_Energy_StencilOutline
dependencies {usedDataInterfaces [], usedDynamicInputs [], usedModules [&SystemState]}
emitters {
 BTR_Energy {
  emitterUpdateScript [
   &EmitterState(Life Cycle Mode=::NewEnumerator1, Loop Behavior=::NewEnumerator1, Loop Duration Mode=::NewEnumerator1, UseLoopDelay=false)
   &SpawnBurst_Instantaneous(Use Spawn Probability=false, UseLoopCountLimit=false)
  ]
  particleSpawnScript [&InitializeParticle(
    Color Mode=::NewEnumerator0
    Sprite Size=<100 94.0754>
    Sprite Size Mode=::NewEnumerator4
    UsePositionOffset=false
    Write Material Random=true
   )]
  particleUpdateScript [&ParticleState(Kill Particles When Lifetime Has Elapsed=false, Let Infinitely Lived Particles Die When Emitter Deactivates=true)]
  rendererClasses [&NiagaraSpriteRendererProperties]
  renderers [{rendererClass=&NiagaraSpriteRendererProperties}]
 }
 BTR_StencilGlow {
  emitterUpdateScript [
   &EmitterState(Life Cycle Mode=::NewEnumerator1, Loop Behavior=::NewEnumerator1, Loop Duration Mode=::NewEnumerator1, UseLoopDelay=false)
   &SpawnBurst_Instantaneous(Use Spawn Probability=false, UseLoopCountLimit=false)
  ]
  particleSpawnScript [&InitializeParticle(Sprite Size=<100 94.0754>, Sprite Size Mode=::NewEnumerator4, UsePositionOffset=false, Write Material Random=true)]
  particleUpdateScript [
   &ParticleState(Kill Particles When Lifetime Has Elapsed=false, Let Infinitely Lived Particles Die When Emitter Deactivates=true)
   &ScaleColor(Scale Alpha=5, ScaleA=true, ScaleRGB=true)
  ]
  rendererClasses [&NiagaraSpriteRendererProperties]
  renderers [{rendererClass=&NiagaraSpriteRendererProperties}]
 }
 BTR_StencilOutline {
  emitterUpdateScript [
   &EmitterState(Life Cycle Mode=::NewEnumerator1, Loop Behavior=::NewEnumerator1, Loop Duration Mode=::NewEnumerator1, UseLoopDelay=false)
   &SpawnBurst_Instantaneous(Use Spawn Probability=false, UseLoopCountLimit=false)
  ]
  particleSpawnScript [&InitializeParticle(
    Color Mode=::NewEnumerator0
    Sprite Size=<100 94.0754>
    Sprite Size Mode=::NewEnumerator4
    UsePositionOffset=false
    Write Material Random=true
   )]
  particleUpdateScript [&ParticleState(Kill Particles When Lifetime Has Elapsed=false, Let Infinitely Lived Particles Die When Emitter Deactivates=true)]
  rendererClasses [&NiagaraSpriteRendererProperties]
  renderers [{rendererClass=&NiagaraSpriteRendererProperties}]
 }
}
summary {
 systemName=NS_BTR_Energy_StencilOutline
 emitters [
  {emitterName=BTR_Energy, rendererClasses [&NiagaraSpriteRendererProperties]}
  {emitterName=BTR_StencilOutline, rendererClasses [&NiagaraSpriteRendererProperties]}
  {emitterName=BTR_StencilGlow, rendererClasses [&NiagaraSpriteRendererProperties]}
 ]
 userVariables []
}
